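# Conversation graph for the Robo specialist NPC, covering its greeting dialogue and the menus reachable from it.
# The Robo entry tag leads to one Speech node with a default line and contextual lines for first-ever meetings, two specific player outfits, and per-match conversation counts.
# That node branches through link nodes to the shared NPC library quests, basic services and Ch5S1.D3.Q8 entry points, keeping those flows outside this asset.
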ConversationDatabase /NPCLibrary/NPCs/Robo/NPCLibrary_Conversation_Robo
entry_points [{tag=Conversation.Entry.Specialist.Robo, destinations [#1]}]
nodes [
 ConversationEntryPointNode_7 {
  id=#1
  class=ConversationEntryPointNode
  outputs [#2]
  properties {entryTag {tagName=Conversation.Entry.Specialist.Robo}}
 }
 Speech_C_4 {
  id=#2
  class=Speech_C
  outputs [#3, #4, #5]
  properties {
   nodeName=""
   generalConfig {
    defaultMessage=This Island's in chaos. But chaos means more work for someone like me.
    contextualMessages [
     {
      message="If you're good at something, never do it for free."
      contextRequirements [{requirement {class=FortControllerRequirement_IsFirstEverConversationWithNPC}}]
      priority {curve {rowName=Tandem.Speech.Priority.FirstEverInteraction}}
     }
     {
      message="Hope's not here. Leave me out of your family drama, thanks."
      contextRequirements [{requirement {class=FortControllerRequirement_HasCID, properties {allowedCIds [/BRCosmetics/Athena/Items/Cosmetics/Characters/Character_MadameMoth_Posh]}}}]
      priority {curve {rowName=Tandem.Speech.Priority.OutfitReaction}}
     }
     {
      message=You're never gonna take down The Society if you won't even take down your sister.
      contextRequirements [{requirement {
         class=FortControllerRequirement_HasCID
         properties {allowedCIds [&Character_VoidRedemption_Rebel]}
        }}]
      priority {curve {rowName=Tandem.Speech.Priority.OutfitReaction}}
     }
     {
      message=This Island's in chaos. But chaos means more work for someone like me.
      contextRequirements [{requirement {
         class=FortControllerRequirement_CheckNPCMatchConversationCount
         properties {nPCUniqueIdTag {tagName=AISpawnerData.Type.Specialist.Supply.Robo}}
        }}]
      priority {curve {rowName=None}}
     }
     {
      message=The Society? The Underground? More like a family feud.
      contextRequirements [{requirement {
         class=FortControllerRequirement_CheckNPCMatchConversationCount
         properties {nPCUniqueIdTag {tagName=AISpawnerData.Type.Specialist.Supply.Robo}}
        }}]
      priority {curve {rowName=None}}
     }
     {
      message=My friend Hope lets me stay here between gigs. It's a good spot for laying low.
      contextRequirements [{requirement {
         class=FortControllerRequirement_CheckNPCMatchConversationCount
         properties {nPCUniqueIdTag {tagName=AISpawnerData.Type.Specialist.Supply.Robo}}
        }}]
      priority {curve {rowName=None}}
     }
     {
      message="This place belongs to Hope, the leader of the Underground."
      contextRequirements [{requirement {
         class=FortControllerRequirement_CheckNPCMatchConversationCount
         properties {nPCUniqueIdTag {tagName=AISpawnerData.Type.Specialist.Supply.Robo}}
        }}]
      priority {curve {rowName=None}}
     }
     {
      message=Don't tell The Society about this place. It's Hope's secret getaway.
      contextRequirements [{requirement {
         class=FortControllerRequirement_CheckNPCMatchConversationCount
         properties {nPCUniqueIdTag {tagName=AISpawnerData.Type.Specialist.Supply.Robo}}
        }}]
      priority {curve {rowName=None}}
     }
     {
      message=The Society are getting nervous. Hope and Jones are making progress.
      contextRequirements [{requirement {
         class=FortControllerRequirement_CheckNPCMatchConversationCount
         properties {nPCUniqueIdTag {tagName=AISpawnerData.Type.Specialist.Supply.Robo}}
        }}]
      priority {curve {rowName=None}}
     }
    ]
   }
  }
 }
 ConversationLinkNode_2 {id=#5, class=ConversationLinkNode, properties {remoteEntryTag {tagName=Conversation.Entry.Ch5S1.D3.Q8}}}
 ConversationLinkNode_0 {id=#4, class=ConversationLinkNode, properties {remoteEntryTag {tagName=Conversation.Entry.BasicServices}}}
 ConversationLinkNode_1 {id=#3, class=ConversationLinkNode, properties {remoteEntryTag {tagName=Conversation.Entry.NPCLibrary.Quests}}}
]
